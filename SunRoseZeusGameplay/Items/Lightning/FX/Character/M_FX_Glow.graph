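# Unlit additive glow material for Niagara sprite and mesh particles in the Lightning item's character FX.
# A UV-centred SphereMask, with radius and hardness both from parameter T's red channel, is sharpened by a power and tinted by particle colour.
# Particle alpha feeds a DepthFade with a parameterised distance.
# Separate branches compute camera-space XY of the particle position and the world position behind translucency, with no consumer shown.
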Material /SunRoseZeusGameplay/Items/Lightning/FX/Character/M_FX_Glow
graph {
 properties {
  bUsedWithNiagaraMeshParticles=true
  bUsedWithNiagaraSprites=true
  blendMode=BLEND_Additive
  pixelDepthOffsetMode=PDOM_Legacy
  shadingModel=MSM_Unlit
  usedShadingModels=Unlit
 }
 nodes [
  ComponentMask(None=TransformPosition(None=MaterialFunctionCall() {materialFunction=&WorldPositionBehindTranslucency}.Result) {transformSourceType=TRANSFORMPOSSOURCE_World, transformType=TRANSFORMPOSSOURCE_Camera}) {g=true, r=true}
  ComponentMask(None=TransformPosition(None=ParticlePositionWS()) {transformSourceType=TRANSFORMPOSSOURCE_World, transformType=TRANSFORMPOSSOURCE_Camera}) {g=true, r=true}
  n1 = ParticleColor()
  DepthFade(Opacity=@n1.A, FadeDistance=ScalarParameter() {defaultValue=5, parameterName=awfa})
  n2 = TextureCoordinate()
  n3 = Constant() {r=0.5}
  n4 = VectorParameter() {parameterName=T, defaultValue=<rgba 0.5 0 0 1>}
  Multiply(A=@n1.RGB, B=Power(Base=SphereMask(A=@n2, B=@n3, Radius=@n4.R, Hardness=@n4.R)) {constExponent=4})
 ]
}
